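; Physics balloon actor that attaches to a weld target and lifts it with a constant upward force.
; Cluster union targets get a locked physics constraint; others use plain attachment.
; A timer refreshes the force every 0.1 seconds, scaled by a height curve to stop targets rocketing away.
; It pops on lifetime expiry, target death or EndPlay, and through a delayed height and map-bounds check.
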(event Custom|MaxHeightCheckToPopBalloon
  (bind (_returnvalue_x _returnvalue_y _returnvalue_z) (Math|Vector|BreakVector (Transformation|GetActorLocation (Variables|Default|GetWeldTarget))))
  (bind (_returnvalue_x_2 _returnvalue_y_2) (Math|Vector|BreakVector (Transformation|GetActorLocation (Variables|Default|GetWeldTarget))))
  ;[0.1136 0.9583 0 0.5] To avoid players getting too high, auto-pop on height check. Also check based on map extents
  (if (> _returnvalue_z (ScalableFloat|GetValueatLevel (Variables|Default|GetMaxHeightCheckToAutoPop)))
    _
    (elif (ScalableFloat|GetValueasBool (Variables|Default|GetEnableCoordCheckToAutoPop))
      (if (Math|Boolean|ORBoolean (> _returnvalue_x_2 (ScalableFloat|GetValueatLevel (Variables|Default|GetMaxPositiveXCoordCheckToAutoPop))) (< _returnvalue_x_2 (ScalableFloat|GetValueatLevel (Variables|Default|GetMaxNegativeXCoordCheckToAutoPop))) (> _returnvalue_y_2 (ScalableFloat|GetValueatLevel (Variables|Default|GetMaxPositiveYCoordCheckToAutoPop))) (< _returnvalue_y_2 (ScalableFloat|GetValueatLevel (Variables|Default|GetMaxNegativeYCoordCheckToAutoPop))))
        _)))
  (Utilities|FlowControl|Delay (ScalableFloat|GetValueatLevel (Variables|Default|GetAutoPopBalloonDelay)))
  (CallFunction|PopBalloon))

(event EventEndPlay
  (params (EndPlayReason :type EEndPlayReason%20Enum :default "Destroyed"))
  (bind _weldtarget (Variables|Default|GetWeldTarget))
  (bind _asbuilding_actor (Utilities|Casting|CastToBuildingActor _weldtarget)
    (:then
      (EventDispatchers|UnbindEventfromOnDied _asbuilding_actor (delegate Custom|OnWeldDied)))
    (:CastFailed
      _))
  (Game|UnbindEventfromOnEndPlay _weldtarget (delegate Custom|OnWeldEndPlay))
  (for _array_element (Actor|GetComponentsByClass self "/Script/Engine.PhysicsConstraintComponent")
    (Components|DestroyComponent _array_element)))

(event EventPlayPopFX
  (params (HitEvent :type Hit%20Result%20Structure))
  (Development|SetHiddeninGame (Variables|Default|GetStaticMesh) true false)
  (|CallParentFunction HitEvent))

(event Custom|OnWeldDied
  (params (DamagedActor :type Actor%20Object%20Reference) (Damage :type Float%20%28single-precision%29) (InstigatedBy :type Controller%20Object%20Reference) (DamageCauser :type Actor%20Object%20Reference) (HitLocation :type Vector) (FHitComponent :type Primitive%20Component%20Object%20Reference) (BoneName :type Name) (Momentum :type Vector))
  (CallFunction|PopBalloon))

(event Custom|OnLifetimeExpired
  (CallFunction|PopBalloon))

(event Custom|OnWeldEndPlay
  (params (Actor :type Actor%20Object%20Reference) (EndPlayReason :type EEndPlayReason%20Enum :default "Destroyed"))
  (CallFunction|PopBalloon))

(event EventBeginPlay
  (bind _weldtarget (Variables|Default|GetWeldTarget))
  (bind _weldtarget_2 (Variables|Default|GetWeldTarget))
  (|CallParentFunction)
  (bind _output_get (Variables|Debug|SetOldPosition (Transformation|GetActorLocation)))
  (Utilities|Casting|CastToClusterUnionActor _weldtarget
    (:then
      (CallFunction|InitializeConstraint _output_get))
    (:CastFailed
      (Transformation|AttachActorToActor _weldtarget "None" "KeepWorld" "KeepWorld" "KeepWorld")))
  (bind _asbuilding_actor (Utilities|Casting|CastToBuildingActor _weldtarget_2)
    (:then
      (EventDispatchers|BindEventtoOnDied _asbuilding_actor (delegate Custom|OnWeldDied)))
    (:CastFailed
      _))
  (Game|BindEventtoOnEndPlay _weldtarget_2 (delegate Custom|OnWeldEndPlay))
  (EventDispatchers|BindEventtoOnDied (delegate Custom|CustomEvent))
  ;[0.1136 0.9583 0 0.5] To avoid players getting too high, auto-pop on height check
  (CallFunction|MaxHeightChecktoPopBalloon)
  (CallFunction|CalcSeaLevel)
  (Utilities|Time|SetTimerbyEvent (delegate Custom|OnLifetimeExpired) (ScalableFloat|GetValueatLevel (Variables|Default|GetMaxLifetime)))
  (bind _returnvalue (FortPhysicsSimulationLibrary|PerformPhysicsObjectValidationChecks (Variables|Default|GetWeldTarget) (Utilities|Struct|MakePhysicsSimValidateObject "Large")))
  (if _returnvalue
    (bind _returnvalue_2 (Components|FindorAddArbitraryComponent (Variables|Default|GetWeldTarget) "/Script/FortniteGame.FortPhysicsConstantForceComponent"))
    (bind _output_get_2 (Variables|Default|SetForceComponent _returnvalue_2))
    (Utilities|IsValid _output_get_2
      (:Is%20Valid
        (Components|Activation|Activate _output_get_2 false)
        (bind _strength (CallFunction|GetForceStrength (Variables|Default|GetWeldTarget)))
        (Variables|/Script/FortniteGame.FortPhysicsConstantForceComponent|SetbActivateFortPhysicsObjectOnBeginConstantForce (Variables|Default|GetActivateTargetPhysics) _output_get_2)
        (bind _outhandle (Class|FortPhysicsConstantForceComponent|BeginConstantForce _output_get_2 (* (Math|Vector|GetUpVector "0, 0, 0") _strength) false true (Math|Vector|Lerp(Vector) (Transformation|GetActorLocation (Variables|Default|GetWeldTarget)) (Transformation|GetActorLocation) (Variables|Default|GetTiltStrength)) true (Variables|Default|GetVelocityChange?) (Variables|Default|GetAttachBone) true))
        (Variables|Default|SetForceHandle _outhandle)
        (Utilities|Time|SetTimerbyEvent (EventDispatchers|CreateEvent :!event Update%20Force) 0.1 true)))))

(event Custom|CustomEvent
  (params (DamagedActor :type Actor%20Object%20Reference) (Damage :type Float%20%28single-precision%29) (InstigatedBy :type Controller%20Object%20Reference) (DamageCauser :type Actor%20Object%20Reference) (HitLocation :type Vector) (FHitComponent :type Primitive%20Component%20Object%20Reference) (BoneName :type Name) (Momentum :type Vector))
  (bind _forcecomponent (Variables|Default|GetForceComponent))
  (Utilities|IsValid _forcecomponent
    (:Is%20Valid
      (Class|FortPhysicsConstantForceComponent|EndConstantForce _forcecomponent (Variables|Default|GetForceHandle) true))))

(fn CalcSeaLevel
  (bind _outactors (Actor|GetAllActorsOfClass "/Script/FortniteGame.FortWaterBodyActor"))
  (bind _array_element (Utilities|Array|ForEachLoopwithBreak _outactors)
    (:LoopBody
      (if (Utilities|Enum|Equal(Enum) (Water|GetWaterBodyType _array_element) "Ocean")
        (Variables|Default|SetSeaLevel (.z (Transformation|GetActorLocation _array_element)))))))

(fn DeubgHitch
  (bind _returnvalue (Utilities|Time|GetTimeSeconds))
  (bind _lastupdatetime (Variables|Default|GetLastUpdateTime))
  (bind _returnvalue_2 (- _returnvalue _lastupdatetime))
  (if (>= _returnvalue_2 (Variables|Default|GetMaxUpdateTime))
    (Development|PrintString (Utilities|String|Append "Balloon Hitch: " (Utilities|String|ToString(Float) _returnvalue_2)))))

(fn GetForceStrength
  (params (Actor :type Actor%20Object%20Reference))
  (bind (_oldposition_x _oldposition_y _oldposition_z) (Math|Vector|BreakVector (Variables|GetOldPosition_X)))
  (Utilities|FlowControl|Sequence
    (:then_0
      (Utilities|IsValid (Variables|Default|GetStrengthOverHeightCurve)
        (:Is%20Valid
          (LocalVariables|SetHeightStrengthMultiplier (Math|Curves|GetFloatValue (Variables|Default|GetStrengthOverHeightCurve) (Math|Float|Max(Float) (- (.z (Transformation|GetActorLocation)) _oldposition_z)))))
        (:Is%20Not%20Valid
          (LocalVariables|SetHeightStrengthMultiplier 1.0))))
    (:then_1
      (return (* (ScalableFloat|GetValueatLevel (Variables|Default|GetBalloonStrength)) (LocalVariables|GetHeightStrengthMultiplier))))))

(fn GetPhysicsCenterOfMassOffset
  (Utilities|Casting|CastToFortAthenaVehicle (Actor|GetAttachParentActor)
    (:then
      (bind _asskeletal_mesh_component (Utilities|Casting|CastToSkeletalMeshComponent (Variables|/Script/Engine.Actor|GetRootComponent (Actor|GetAttachParentActor)))
        (:then
          (return (.COMNudge (Utilities|Struct|BreakBodyInstance (Variables|/Script/Engine.PrimitiveComponent|GetBodyInstance _asskeletal_mesh_component)))))
        (:CastFailed
          _)))
    (:CastFailed
      _))
  (return "0, 0, 0"))

(fn InitializeConstraint
  (params (WorldLocation :type Vector))
  (bind _returnvalue (FortPhysicsSimulationLibrary|CreatePhysicsConstraintBetween self "root" (Variables|Default|GetWeldTarget) (Variables|Default|GetAttachBone) WorldLocation))
  (Physics|Components|PhysicsConstraint|SetAngularSwing1Limit _returnvalue "ACM_Locked" 0.0)
  (Physics|Components|PhysicsConstraint|SetAngularSwing2Limit _returnvalue "ACM_Locked" 0.0)
  (Physics|Components|PhysicsConstraint|SetAngularTwistLimit _returnvalue "ACM_Locked" 0.0)
  (Physics|Components|PhysicsConstraint|SetLinearXLimit _returnvalue "LCM_Locked" 0.0)
  (Physics|Components|PhysicsConstraint|SetLinearYLimit _returnvalue "LCM_Locked" 0.0)
  (Physics|Components|PhysicsConstraint|SetLinearZLimit _returnvalue "LCM_Locked" 0.0)
  (Physics|SetAllBodiesSimulatePhysics (Variables|Default|GetBalloon_SKBudgeted) true))

(fn PopBalloon
  (Utilities|FlowControl|DoOnce false
    (:Completed
      (CallFunction|PlayPopFX)
      (Actor|DestroyActor))))

(fn ShouldApplyForces
  (bind _asfort_athena_vehicle (Utilities|Casting|CastToFortAthenaVehicle (Variables|Default|GetWeldTarget))
    (:then
      (if (Vehicle|HasDriver _asfort_athena_vehicle)
        (return (Vehicle|ClientIsDriver _asfort_athena_vehicle))
        (else
          (return (Networking|IsServer)))))
    (:CastFailed
      (return true))))

(fn UpdateForce
  ;[1 1 1] Update the force periodically to make balloons less powerful as they climb higher so things don't rocket into space
  (bind _strength (CallFunction|GetForceStrength (Variables|Default|GetWeldTarget)))
  (Class|FortPhysicsConstantForceComponent|ModifyConstantForce (Variables|Default|GetForceComponent) (Variables|Default|GetForceHandle) (* (Math|Vector|GetUpVector "0, 0, 0") _strength) false true (Transformation|GetActorLocation) true (Variables|Default|GetVelocityChange?) (Variables|Default|GetAttachBone) false))

(fn ConstructionScript
  (|CallParentFunction))
